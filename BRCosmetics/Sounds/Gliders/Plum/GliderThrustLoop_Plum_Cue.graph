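# Sound cue for the looping thrust and idle audio of the Plum glider cosmetic.
# A mixer layers two looping waves, each driven by a continuous modulator that scales pitch and volume from a named float parameter.
# The thrust loop follows Forwards, rising from near silence, while the idle loop follows AnyDirection and stays more audible at rest.
# Glider thrust attenuation, a cosmetic glider loop properties sheet and a glider sound class govern playback.
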SoundCue /BRCosmetics/Sounds/Gliders/Plum/GliderThrustLoop_Plum_Cue
root=Mixer_0
nodes [
 {id=Mixer_0, children [ModulatorContinuous_0, ModulatorContinuous_1], properties {inputVolume [1,1]}}
 {
  id=ModulatorContinuous_0
  children [WavePlayer_0]
  properties {
   pitchModulationParams {default=0, maxOutput=1.2, minOutput=1, parameterName=Forwards}
   volumeModulationParams {default=0, maxOutput=0.55, minOutput=0.05, parameterName=Forwards}
  }
 }
 {
  id=ModulatorContinuous_1
  children [WavePlayer_1]
  properties {
   pitchModulationParams {default=0, maxOutput=1.2, minOutput=1, parameterName=AnyDirection}
   volumeModulationParams {default=0, maxOutput=0.75, minOutput=0.5, parameterName=AnyDirection}
  }
 }
 {id=WavePlayer_0, properties {bLooping=true, soundWaveAssetPtr=&/Glider_Plum_Thrust_Loop}}
 {id=WavePlayer_1, properties {bLooping=true, soundWaveAssetPtr=&/Glider_Plum_Idle_Loop}}
]
properties {
 attenuationSettings=&Fort_Player_Glider_Thrust_Athena_Att
 audioPropertiesSheet=&APS_Cosmetic_Glider_Loop
 maxDistance=12500
 soundClassObject=&Fort_Gliders_Land
}
